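# Material function for the Party Trooper character that adds a pulsing, tinted emissive effect to incoming material attributes.
# A static switch picks a cropped body heart mask, whose crop corners animate with a time-driven sine to pulse its size, or an uncropped face mask.
# The red channel blends in the line glow, while green suppresses existing emissive and adds a gradient glow, both pulsing from that sine with separate color and intensity controls.
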MaterialFunction /BRCosmetics/Characters/Player/Male/Medium/Bodies/M_MED_Party_Trooper/Materials/MF_Party_Trooper
graph {nodes [
  n1 = GetMaterialAttributes(None=FunctionInput() {inputType=FunctionInput_MaterialAttributes}) {attributeGetTypes [#1, #2]}
  n2 = TextureCoordinate()
  n3 = TextureObjectParameter() {
   group=Party_Trooper
   parameterName=FX Mask
   samplerType=SAMPLERTYPE_Masks
   texture=/BRCosmetics/Characters/Player/Male/Medium/Bodies/M_MED_Party_Trooper/Textures/T_Party_Trooper_A_Heart_Cropped_FX
  }
  n4 = VectorParameter() {group=Party_Trooper, parameterName=PT_FXCrop-UpperLeft, defaultValue=<rgba 0.813 0.371 0 1>}
  n5 = VectorParameter() {group=Party_Trooper, parameterName=PT_CropOffset, defaultValue=<rgba 0 0 0 1>}
  n6 = VectorParameter() {desc=Offset added to both corners to scale Texture, parameterName=PT_CornerOffsetScale, defaultValue=<rgba 0.02 0.02 0 1>}
  n7 = ScalarParameter() {defaultValue=0.01, group=Party_Trooper, parameterName=PT_PulseScale, sliderMax=1}
  n8 = Sine(None=Multiply(A=Time(), B=Constant() {r=0.9}))
  n9 = Constant() {r=5}
  n10 = Saturate(None=Add(A=Multiply(A=@n8, B=@n9), B=@n9))
  n11 = ScalarParameter() {defaultValue=0.6, desc=Move the pivot to upperleft when scaling, group=Party_Trooper, parameterName=PT_UpperLeftCornerMult}
  n12 = Add(A=Add(A=Add(A=@n4.RGB, B=@n5.RGB), B=Multiply(A=@n6.RGB) {constB=-1}), B=Multiply(A=Multiply(A=@n10, B=@n7), B=@n11) {constB=0.5})
  n13 = VectorParameter() {group=Party_Trooper, parameterName=PT_FXCrop-LowerRight, defaultValue=<rgba 0.959 0.517 0 1>}
  n14 = Add(A=Multiply(A=@n10, B=Multiply(A=@n7) {constB=-1}), B=Add(A=@n6.RGB, B=Add(A=@n13.RGB, B=@n5.RGB)))
  n15 = TextureSampleParameter2D(UVs=TextureCoordinate()) {
   parameterName=FX Mask
   samplerType=SAMPLERTYPE_Masks
   texture=/BRCosmetics/Characters/Player/Male/Medium/Bodies/M_MED_Party_Trooper/Textures/T_Party_Trooper_A_Face_FX
  }
  n16 = MaterialFunctionCall(Float2=StaticSwitchParameter(
    True=MaterialFunctionCall!(UVs=@n2, TextureIn=@n3, UpperLeftCorner=@n12, LowerRightCorner=@n14, ExportFloat4) {materialFunction=/Engine/Functions/Engine_MaterialFunctions01/Texturing/TextureCropping}.CroppedMasked
    False=@n15.RGB
   ) {defaultValue=true, group=Party_Trooper, parameterName=Is Body}) {materialFunction=&BreakOutFloat2Components}
  n17 = OneMinus(None=Saturate(None=@n16.G))
  n18 = ScalarParameter() {defaultValue=5, group=Party_Trooper, parameterName=PT_PulseEmsBase-Line}
  n19 = Multiply(A=OneMinus(None=@n10), B=ScalarParameter() {defaultValue=5, group=Party_Trooper, parameterName=PT_PulseEmsMult-Line}) {constB=5}
  n20 = LinearInterpolate(
   A=Multiply(A=@n1.EmissiveColor, B=Power(Base=@n17) {constExponent=10}) {constA=1}
   B=Multiply(
    A=VectorParameter() {group=Party_Trooper, parameterName=PT_PulseEmsColort-Line, defaultValue=<rgba 1 0.1458 0.3049 1>}.RGB
    B=Add(A=@n18, B=@n19)
   )
   Alpha=@n16.R
  )
  n21 = ScalarParameter() {defaultValue=1, group=Party_Trooper, parameterName=PT_PulseEmsBase-Gradient}
  n22 = ScalarParameter() {defaultValue=1, group=Party_Trooper, parameterName=PT_EmsTempo}
  n23 = Add(A=Multiply(A=@n8, B=@n22), B=@n22)
  n24 = Multiply(A=OneMinus(None=Saturate(None=@n23)), B=ScalarParameter() {defaultValue=1, group=Party_Trooper, parameterName=PT_PulseEmsMult-Gradient}) {constB=5}
  n25 = Multiply(
   A=@n16.G
   B=Multiply(
    A=VectorParameter() {group=Party_Trooper, parameterName=PT_PulseEmsColort-Gradient, defaultValue=<rgba 1 0.1458 0.3049 1>}.RGB
    B=Add(A=@n21, B=@n24)
   )
  )
  FunctionOutput(None=SetMaterialAttributes!(MaterialAttributes=@n1.MaterialAttributes, EmissiveColor=Add(A=@n20, B=@n25)) {attributeSetTypes [#1]})
 ]}
