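# Master material for the Party Trooper medium male player body, a Default Lit skeletal mesh material supporting morph targets and mesh deformers.
# It uses material attributes, passing the shared character PreFX result through MF_Party_Trooper and into the shared PostFX function.
# PostFX's AO, Distance Blend, Fresnel, Gmap and Pre Skinned Local Position inputs are all wired to the PreFX AO output.
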Material /BRCosmetics/Characters/Player/Male/Medium/Bodies/M_MED_Party_Trooper/Materials/M_MED_Party_Trooper_MASTER
graph {
 properties {
  bUseMaterialAttributes=true
  bUsedWithMeshDeformer=true
  bUsedWithMorphTargets=true
  bUsedWithSkeletalMesh=true
  pixelDepthOffsetMode=PDOM_Legacy
  usedShadingModels=Default Lit
 }
 nodes [
  n1 = MaterialFunctionCall() {materialFunction=&BaseFN_CharacterFunction_PreFX}
  MaterialFunctionCall(
   AO=@n1.AO
   Distance Blend=@n1.AO
   Fresnel=@n1.AO
   Gmap=@n1.AO
   Pre Skinned Local Position=@n1.AO
   Material=MaterialFunctionCall!(In=@n1.Result) {materialFunction=&/MF_Party_Trooper}.Result
  ) {materialFunction=&BaseFN_CharacterFunction_PostFX}
 ]
}
